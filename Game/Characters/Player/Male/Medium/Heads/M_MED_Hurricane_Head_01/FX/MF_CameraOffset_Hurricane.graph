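# Material function for the Hurricane male player head effects that offsets world position toward or away from the camera, likely for its sprites.
# It feeds the scalar parameter HC_Sprites_CameraOffset into the engine CameraOffset function and outputs the clamped camera offset.
# Exposing the amount as a named parameter lets materials using the function tune the offset per instance.
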MaterialFunction /Game/Characters/Player/Male/Medium/Heads/M_MED_Hurricane_Head_01/FX/MF_CameraOffset_Hurricane
graph {nodes [FunctionOutput(None=MaterialFunctionCall(Offset Amount=ScalarParameter() {defaultValue=21, parameterName=HC_Sprites_CameraOffset}) {materialFunction=/Engine/Functions/Engine_MaterialFunctions02/WorldPositionOffset/CameraOffset}.Clamped Camera Offset)]}
