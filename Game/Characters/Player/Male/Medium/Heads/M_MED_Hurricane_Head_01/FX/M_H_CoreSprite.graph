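# Additive, unlit, two-sided sprite material for the glowing core effect on the Hurricane player head, usable on Niagara sprites.
# Panning noise, offset per particle via dynamic parameter, varies a sphere mask's radius to form a distorted radial glow.
# The powered mask is tinted by core color and optional particle color, then boosted through an emissive-adjustment function.
# A camera-offset function is included, and depth fade softens opacity except on mobile shading.
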Material /Game/Characters/Player/Male/Medium/Heads/M_MED_Hurricane_Head_01/FX/M_H_CoreSprite
graph {
 properties {
  bUsedWithInstancedStaticMeshes=true
  bUsedWithNiagaraSprites=true
  blendMode=BLEND_Additive
  pixelDepthOffsetMode=PDOM_Legacy
  shadingModel=MSM_Unlit
  twoSided=true
  usedShadingModels=Unlit
 }
 nodes [
  n1 = Constant() {r=1}
  n2 = ParticleColor()
  n3 = ScalarParameter() {parameterName=UseParticleColor}
  n4 = TextureCoordinate()
  n5 = Add(A=TextureCoordinate(), B=DynamicParameter() {paramNames [UVOffset,Param2,Param3,Param4]}.UVOffset)
  n6 = ComponentMask(None=VectorParameter() {parameterName=HC_CoreDistortionNoise_Tiling, defaultValue=<rgba 0.5 0.5 0.5 1>}.RGB) {g=true, r=true}
  n7 = VectorParameter() {parameterName=HC_CoreDistortionNoise_PanSpeedAndDirection, defaultValue=<rgba 0.15 0.15 1 1>}
  n8 = TextureSampleParameter2D(UVs=Panner(Coordinate=Multiply(A=@n5, B=@n6), Speed=ComponentMask(None=@n7.RGB) {g=true, r=true})) {parameterName=DistortTexture, samplerType=SAMPLERTYPE_LinearColor, texture=&T_NoisePack_LinearColor}
  n9 = ScalarParameter() {defaultValue=1, parameterName=HC_CoreDistortionStrength}
  n10 = SphereMask(
   A=Subtract(A=Multiply(A=@n4) {constB=2})
   B=Constant2Vector()
   Radius=LinearInterpolate(Alpha=Multiply(A=@n8.R, B=@n9)) {constA=0.5, constB=1.5}
   Hardness=ScalarParameter() {defaultValue=0.1, parameterName=SphereMask-Hardness}
  )
  n11 = Multiply(
   A=Multiply(A=LinearInterpolate(A=@n1, B=@n2.RGB, Alpha=@n3), B=VectorParameter() {parameterName=HC_Core_Color, defaultValue=<rgba 1 0.4 0.05 0>}.RGB)
   B=Clamp(None=Power(Base=@n10) {constExponent=4})
  )
  n12 = ScalarParameter() {defaultValue=25, parameterName=HC_CoreGlowStrength}
  MaterialFunctionCall(Emissive=Power(Base=Multiply(A=@n11, B=@n12))) {materialFunction=&MF_FX_EmissiveAdjustments}
  Reroute(None=MaterialFunctionCall() {materialFunction=&/MF_CameraOffset_Hurricane}.Result)
  n13 = LinearInterpolate(A=Constant() {r=1}, B=ParticleColor().A, Alpha=@n3)
  n14 = DepthFade(Opacity=@n13, FadeDistance=ScalarParameter() {parameterName=DepthFade-Distance})
  ShadingPathSwitch(Default=@n14, Deferred=@n14, Mobile=@n13) {desc=Migrated from FeatureLevelSwitch}
 ]
}
